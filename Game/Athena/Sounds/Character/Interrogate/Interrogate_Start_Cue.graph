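# Sound cue for the start of an interrogation in the Athena mode, part of the character sound set.
# A single wave player feeds the Interrogate_Start_01 wave straight to the output, with no randomization or mixing.
# It routes through the third-person enemy player foley sound class and borrows the third-person DBNO shuffle attenuation for spatial falloff.
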SoundCue /Game/Athena/Sounds/Character/Interrogate/Interrogate_Start_Cue
root=WavePlayer_0
nodes [{id=WavePlayer_0, properties {soundWaveAssetPtr=&/Interrogate_Start_01}}]
properties {
 attenuationSettings=/Game/Sounds/Attenuation/Fort_Player_DBNO_Shuffle_3P_Athena_Att
 maxDistance=1500
 soundClassObject=&Fort_Player_Foley_3P_Enemy_Athena
 volumeMultiplier=1
}
